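# Base behavior tree for Athena husk AI that chases the blackboard GoalActor, with no attack task despite its "Attack Or Hunt" branch.
# With a goal set, it pauses briefly if that actor is airborne, then waits when already there or moves toward it using a no-water nav filter.
# A failed move still succeeds if the goal is airborne.
# Otherwise short waits act as fallbacks, including when no goal exists.
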BehaviorTree /Game/Athena/AI/BehaviorTrees/BT_AthenaHuskBase
blackboard=/Game/AI/Blackboards/FortAIBB_Default
root {
 name=Selector
 class=BTComposite_Selector
 children [
  Attack Or Hunt {
   class=BTComposite_Selector
   children [
    Sequence {
     class=BTComposite_Sequence
     children [
      Wait for Goal Pawn to Fall Off {
       class=BTTask_Wait
       decorators [
        Can Actor be considered Airborne {
         class=FortBTDecorator_CanBeConsideredAirborne
         properties {
          bInverseCondition=false
          flowAbortMode=None
          nodeName=Can Actor be considered Airborne
          blackboardKey {bNoneIsAllowedValue=false, selectedKeyName=GoalActor}
         }
        }
        Force Success {class=BTDecorator_ForceSuccess}
       ]
       properties {nodeName=Wait for Goal Pawn to Fall Off, randomDeviation {defaultValue=0.1}, waitTime {defaultValue=0.25}}
      }
      Selector {
       class=BTComposite_Selector
       children [
        Wait {
         class=BTTask_Wait
         decorators [At Goal Actor {
           class=BTDecorator_IsAtLocation
           properties {
            acceptableRadius=50
            bInverseCondition=false
            flowAbortMode=None
            geometricDistanceType=Distance3D
            nodeName=At Goal Actor
            bPathFindingBasedTest {defaultValue=true, key=None}
            bUseNavAgentGoalLocation {defaultValue=true, key=None}
            blackboardKey {bNoneIsAllowedValue=false, selectedKeyName=GoalActor}
            parametrizedAcceptableRadius {dataBinding=null, dataField=None, defaultValue=0}
           }
          }]
         properties {waitTime {defaultValue=1}}
        }
        Move to Goal Actor {
         class=FortBTTask_GameMoveTo
         properties {
          bDeimosFlavor=true
          bFinishMoveOnOverlap=false
          bStopAtGoal=true
          nodeName=Move to Goal Actor
          acceptableRadius {defaultValue=100}
          bReachTestIncludesAgentRadius {defaultValue=false}
          bReachTestIncludesGoalRadius {defaultValue=false}
          blackboardKey {selectedKeyName=GoalActor}
          filterClass {defaultValue=/Game/Athena/AI/Phoebe/NavFilter/NavFilter_Phoebe_NoWater.NavFilter_Phoebe_NoWater_C}
          navFilterTag {tagName=None}
         }
        }
        Succeed {
         class=BTTask_FinishWithResult
         children []
         decorators [Can Actor be considered Airborne {
           class=FortBTDecorator_CanBeConsideredAirborne
           properties {
            bInverseCondition=false
            flowAbortMode=None
            nodeName=Can Actor be considered Airborne
            blackboardKey {bNoneIsAllowedValue=false, selectedKeyName=GoalActor}
           }
          }]
         properties {
          bIgnoreRestartSelf=false
          nodeName=Succeed
          result {defaultValue=0, enumType=/Script/AIModule.EBTNodeResult, key=None, nativeEnumTypeName=EBTNodeResult}
         }
         services []
        }
       ]
       decorators []
       properties {
        children [
         {
          childComposite=null
          childTask=&self:BTTask_Wait_2
          decorators [&self:BTDecorator_IsAtLocation_0]
         }
         {childComposite=null, childTask=&self:FortBTTask_GameMoveTo_2, decorators []}
         {
          childComposite=null
          childTask=&self:BTTask_FinishWithResult_5
          decorators [&self:FortBTDecorator_CanBeConsideredAirborne_0]
         }
        ]
        services []
       }
       services []
      }
     ]
     decorators []
     properties {children [
       {
        childTask=&self:BTTask_Wait_1
        decorators [
         &self:FortBTDecorator_CanBeConsideredAirborne_3
         &self:BTDecorator_ForceSuccess_3
        ]
       }
       {childComposite=&self:BTComposite_Selector_6, childTask=null}
      ]}
    }
    Wait {class=BTTask_Wait, properties {randomDeviation {defaultValue=0.25}, waitTime {defaultValue=0.5}}}
   ]
   decorators [Has Goal Actor? {
     class=BTDecorator_Blackboard
     properties {
      arithmeticOperation=Equal
      intValue=0
      nodeName=Has Goal Actor?
      notifyObserver=ValueChange
      stringValue=""
      blackboardKey {selectedKeyName=GoalActor}
     }
    }]
   properties {
    nodeName=Attack Or Hunt
    children [
     {childComposite=&self:BTComposite_Sequence_3, childTask=null, decorators []}
     {childComposite=null, childTask=&self:BTTask_Wait_12, decorators []}
    ]
    services []
   }
   services []
  }
  Wait {class=BTTask_Wait, properties {waitTime {defaultValue=1}}}
 ]
 decorators []
 properties {
  children [
   {
    childComposite=&self:BTComposite_Selector_48
    childTask=null
    decorators [&self:BTDecorator_Blackboard_122]
   }
   {childComposite=null, childTask=&self:BTTask_Wait_0, decorators []}
  ]
  services []
 }
 services []
}
