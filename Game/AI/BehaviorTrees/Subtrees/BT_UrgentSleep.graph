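# Behavior subtree that sends an AI back to its spawn point to sleep, gated by an Allowed To Sleep check that aborts Both, with a service skipping unperceived fallback goals.
# While a montage plays, a short Wait takes priority.
# Otherwise it moves at medium urgency to SpawnLocation using a no-smashing nav filter, faces SpawnRotation, pauses, then sleeps.
# Force Success on the return sequence lets Sleep run even if movement or rotation fails.
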BehaviorTree /Game/AI/BehaviorTrees/Subtrees/BT_UrgentSleep
blackboard=/Game/AI/Blackboards/FortAIBB_Default
root {
 name=Selector
 class=BTComposite_Selector
 children [
  Selector {
   class=BTComposite_Selector
   children [Wait {class=BTTask_Wait, properties {waitTime {defaultValue=0.5}}}]
   decorators [Is Playing Montage {
     class=IsPlayingMontage_C
     properties {
      bCheckConditionOnlyBlackBoardChanges=false
      bInverseCondition=false
      bShowPropertyDetails=true
      customDescription=""
      flowAbortMode=None
      nodeName=Is Playing Montage
     }
    }]
   properties {children [{childComposite=null, childTask=&self:BTTask_Wait_0, decorators []}], services []}
   services []
  }
  Return to Spawn point and Sleep {
   class=BTComposite_Sequence
   children [
    Sequence {
     class=BTComposite_Sequence
     children [
      FortMoveTo {
       class=FortBTTask_GameMoveTo
       properties {
        bDeimosFlavor=false
        bFinishMoveOnOverlap=true
        bStopAtGoal=true
        nodeName=FortMoveTo
        acceptableRadius {defaultValue=5}
        bReachTestIncludesAgentRadius {defaultValue=true}
        bReachTestIncludesGoalRadius {defaultValue=true}
        blackboardKey {selectedKeyName=SpawnLocation}
        filterClass {defaultValue=/Script/FortniteAI.FortNavigationFilter_NoSmashing}
        navFilterTag {tagName=None}
       }
      }
      Rotate to face BB entry {class=BTTask_RotateToFaceBBEntry, decorators [], properties {blackboardKey {selectedKeyName=SpawnRotation}, precision {defaultValue=5}}}
      Wait {class=BTTask_Wait, properties {waitTime {defaultValue=1}}}
     ]
     decorators [Force Success {class=BTDecorator_ForceSuccess}]
     properties {children [
       {childTask=&self:FortBTTask_GameMoveTo_0}
       {childTask=&self:BTTask_RotateToFaceBBEntry_0}
       {childTask=&self:BTTask_Wait_1}
      ]}
    }
    Sleep {class=FortBTTask_Sleep, children [], decorators [], properties {bIgnoreRestartSelf=false, nodeName=Sleep}, services []}
   ]
   decorators []
   properties {
    nodeName=Return to Spawn point and Sleep
    children [
     {
      childComposite=&self:BTComposite_Sequence_123
      childTask=null
      decorators [&self:BTDecorator_ForceSuccess_31]
     }
     {childTask=&self:FortBTTask_Sleep_0}
    ]
    services [&self:FortBTContext_MoveUrgency_12]
   }
   services [Move Urgency - Medium {
     class=FortBTContext_MoveUrgency
     properties {
      bCallTickOnSearchStart=false
      bRestartTimerOnEachActivation=false
      interval=0.5
      nodeName=Move Urgency - Medium
      randomDeviation=0.1
      enabled {defaultValue=true, key=None}
      moveUrgencyValue {defaultValue=2, enumType=/Script/FortniteGame.EFortMovementUrgency, key=None, nativeEnumTypeName=EFortMovementUrgency}
     }
    }]
  }
 ]
 decorators [Allowed To Sleep? {
   class=AllowedToSleep_C
   properties {
    bCheckConditionOnlyBlackBoardChanges=false
    bInverseCondition=false
    bShowPropertyDetails=true
    customDescription=""
    flowAbortMode=Both
    nodeName=Allowed To Sleep?
   }
  }]
 properties {
  children [
   {
    childComposite=&self:BTComposite_Selector_1
    childTask=null
    decorators [&self:IsPlayingMontage_C_0]
   }
   {childComposite=&self:BTComposite_Sequence_42, childTask=null, decorators []}
  ]
  services [&self:FortBTContext_SkipNotPerceivedGoals_0]
 }
 services ["Skip Fallback Goal: AllEnemies" {
   class=FortBTContext_SkipNotPerceivedGoals
   properties {
    bCallTickOnSearchStart=false
    bRestartTimerOnEachActivation=false
    interval=0.5
    nodeName=Skip Fallback Goal: AllEnemies
    randomDeviation=0.1
    enabled {defaultValue=true, key=None}
   }
  }]
}
