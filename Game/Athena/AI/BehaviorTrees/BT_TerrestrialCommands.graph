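# Behavior tree that runs ground NPC responses to ping commands: GoTo and HoldPosition.
# GoTo fights a goal pawn near the ping if the NPC can attack, otherwise moves to the ping and waits; HoldPosition just waits.
# Combat uses tag-driven melee target slots to gate a reposition, face and approach loop, while services trigger primary and lunge attacks.
# Fallbacks reposition for slots, move to the goal, or wander nearby.
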BehaviorTree /Game/Athena/AI/BehaviorTrees/BT_TerrestrialCommands
blackboard=/Game/Athena/AI/Blackboards/FortAIBB_NPC_Default
root {
 name=Selector
 class=BTComposite_Selector
 children [
  Selector {
   class=BTComposite_Selector
   children [
    Selector {
     class=BTComposite_Selector
     children [
      Selector {
       class=BTComposite_Selector
       children [
        Selector {
         class=BTComposite_Selector
         children [Selector {
           class=BTComposite_Selector
           children [Selector {
             class=BTComposite_Selector
             children [Selector {
               class=BTComposite_Selector
               children [
                GameplayAbility {
                 class=FortBTTask_ExecuteGameplayAbility
                 decorators [
                  TimeLimit {
                   class=BTDecorator_TimeLimit
                   properties {bInverseCondition=false, flowAbortMode=Self, nodeName=TimeLimit, timeLimit {defaultValue=1, key=None}}
                  }
                  Compare Gameplay Ability Distances {
                   class=FortBTDecorator_GameplayAbility_CompareDistance
                   properties {
                    bInverseCondition=false
                    flowAbortMode=None
                    nodeName=Compare Gameplay Ability Distances
                    activeAbilityTagsToSkipTestingValue {key=None, defaultValue {gameplayTags []}}
                    distanceComparisons [{
                      addedOffset=0
                      bUseAddedOffset=false
                      bUseOverriddenValue=false
                      comparisonType=ThreeDimensions
                      operator=Less
                      overriddenValue=0
                      distanceDataTags {gameplayTags [{tagName=NPC.Ability.MinPreferredRange}]}
                     }]
                    gameplayAbilityTagContainer {key=None, defaultValue {gameplayTags [{tagName=NPC.Ability.Attack.Melee.Primary}]}}
                    targetValue {baseClass=&Actor, defaultValue=null, key=GoalActor}
                   }
                  }
                  Has Gameplay Ability {
                   class=FortBTDecorator_GameplayAbility_HasGameplayAbility
                   properties {gameplayAbilityTagContainer {defaultValue {gameplayTags [{tagName=NPC.Ability.NonAttack.Reposition.MinPreferredRange.Melee.Backup}]}}}
                  }
                  Can Activate Gameplay Ability {
                   class=FortBTDecorator_GameplayAbility_CanActivate
                   properties {
                    flowAbortMode=None
                    gameplayAbilityTagContainer {defaultValue {gameplayTags [{tagName=NPC.Ability.NonAttack.Reposition.MinPreferredRange.Melee.Backup}]}}
                    targetValue {key=AI_Ping_Actor}
                   }
                  }
                 ]
                 properties {
                  nodeName=GameplayAbility
                  gameplayAbilityTagContainer {defaultValue {gameplayTags [{tagName=NPC.Ability.NonAttack.Reposition.MinPreferredRange.Melee.Backup}]}}
                 }
                }
                Selector {
                 class=BTComposite_Selector
                 children [
                  Rotate to face BB entry {
                   class=BTTask_RotateToFaceBBEntry
                   decorators [NOT Rotated To Attack Target {
                     class=FortBTDecorator_GameplayAbility_IsRotatedToAttackTarget
                     properties {
                      bInverseCondition=true
                      flowAbortMode=None
                      nodeName=NOT Rotated To Attack Target
                      activeAbilityTagsToSkipTestingValue {key=None, defaultValue {gameplayTags []}}
                      gameplayAbilityTagContainer {key=None, defaultValue {gameplayTags [{tagName=NPC.Ability.Attack.Melee.Primary}]}}
                      targetValue {baseClass=&Actor, defaultValue=null, key=GoalActor}
                     }
                    }]
                   properties {blackboardKey {selectedKeyName=GoalActor}, precision {defaultValue=10}}
                  }
                  Wait {class=BTTask_Wait, properties {waitTime {defaultValue=0.1}}}
                 ]
                 decorators [Can Hit with Gameplay Ability {
                   class=FortBTDecorator_GameplayAbility_CanHitTarget
                   properties {
                    bInverseCondition=false
                    flowAbortMode=None
                    nodeName=Can Hit with Gameplay Ability
                    activeAbilityTagsToSkipTestingValue {key=None, defaultValue {gameplayTags []}}
                    gameplayAbilityTagContainer {key=None, defaultValue {gameplayTags [{tagName=NPC.Ability.Attack.Melee.Primary}]}}
                    targetValue {baseClass=&Actor, defaultValue=null, key=GoalActor}
                    useIdealYawRotationToTargetValue {defaultValue=true, key=None}
                   }
                  }]
                 properties {
                  children [
                   {
                    childComposite=null
                    childTask=&self:BTTask_RotateToFaceBBEntry_1
                    decorators [&self:FortBTDecorator_GameplayAbility_IsRotatedToAttackTarget_1]
                   }
                   {childComposite=null, childTask=&self:BTTask_Wait_8, decorators []}
                  ]
                  services []
                 }
                 services []
                }
                Sequence {
                 class=BTComposite_Sequence
                 children [
                  Selector {
                   class=BTComposite_Selector
                   children [
                    GameplayAbility {
                     class=FortBTTask_ExecuteGameplayAbility
                     decorators [
                      Has Gameplay Ability {
                       class=FortBTDecorator_GameplayAbility_HasGameplayAbility
                       properties {gameplayAbilityTagContainer {defaultValue {gameplayTags [{tagName=NPC.Ability.NonAttack.Reposition.MinPreferredRange.Melee.MoveToTarget}]}}}
                      }
                      Can Activate Gameplay Ability {
                       class=FortBTDecorator_GameplayAbility_CanActivate
                       properties {
                        flowAbortMode=None
                        gameplayAbilityTagContainer {defaultValue {gameplayTags [{tagName=NPC.Ability.NonAttack.Reposition.MinPreferredRange.Melee.MoveToTarget}]}}
                        targetValue {key=AI_Ping_Actor}
                       }
                      }
                      Is NOT Closest Nearby Pawn {
                       class=FortBTDecorator_GameplayAbility_IsClosestPawnInRange
                       properties {
                        bInverseCondition=true
                        flowAbortMode=None
                        nodeName=Is NOT Closest Nearby Pawn
                        activeAbilityTagsToSkipTestingValue {key=None, defaultValue {gameplayTags []}}
                        filterAIPawns {defaultValue=true, key=None}
                        filterPlayerPawns {defaultValue=false, key=None}
                        gameplayAbilityTagContainer {key=None, defaultValue {gameplayTags [{tagName=NPC.Ability.NonAttack.Reposition.MinPreferredRange.Melee.MoveToTarget}]}}
                        nearbyPawnDistanceValue {defaultValue=3000, key=None}
                        targetValue {baseClass=&Actor, defaultValue=null, key=GoalActor}
                       }
                      }
                     ]
                     properties {
                      nodeName=GameplayAbility
                      gameplayAbilityTagContainer {defaultValue {gameplayTags [{tagName=NPC.Ability.NonAttack.Reposition.MinPreferredRange.Melee.MoveToTarget}]}}
                     }
                    }
                    FortMoveTo {
                     class=FortBTTask_GameMoveTo
                     properties {
                      bDeimosFlavor=false
                      bFinishMoveOnOverlap=true
                      bStopAtGoal=false
                      nodeName=FortMoveTo
                      acceptableRadius {defaultValue=30}
                      bReachTestIncludesAgentRadius {defaultValue=true}
                      bReachTestIncludesGoalRadius {defaultValue=true}
                      blackboardKey {selectedKeyName=GoalActor}
                      filterClass {defaultValue=&NavFilter_NPC_AnyTraversable_C}
                      navFilterTag {tagName=NavFilter.NPC.DefaultMove}
                     }
                    }
                   ]
                   decorators []
                   properties {
                    children [
                     {
                      childComposite=null
                      childTask=&self:FortBTTask_ExecuteGameplayAbility_4
                      decorators [
                       &self:FortBTDecorator_GameplayAbility_HasGameplayAbility_6
                       &self:FortBTDecorator_GameplayAbility_CanActivate_4
                       &self:FortBTDecorator_GameplayAbility_IsClosestPawnInRange_0
                      ]
                     }
                     {childComposite=null, childTask=&self:FortBTTask_GameMoveTo_5, decorators []}
                    ]
                    services []
                   }
                   services []
                  }
                  Wait {class=BTTask_Wait, properties {waitTime {defaultValue=0.2}}}
                 ]
                 decorators [Compare Gameplay Ability Distances {
                   class=FortBTDecorator_GameplayAbility_CompareDistance
                   properties {
                    bInverseCondition=false
                    flowAbortMode=Both
                    nodeName=Compare Gameplay Ability Distances
                    activeAbilityTagsToSkipTestingValue {key=None, defaultValue {gameplayTags []}}
                    distanceComparisons [{
                      addedOffset=0
                      bUseAddedOffset=false
                      bUseOverriddenValue=false
                      comparisonType=ThreeDimensions
                      operator=Greater
                      overriddenValue=0
                      distanceDataTags {gameplayTags [{tagName=NPC.Ability.MinPreferredRange}]}
                     }]
                    gameplayAbilityTagContainer {key=None, defaultValue {gameplayTags [{tagName=NPC.Ability.Attack.Melee.Primary}]}}
                    targetValue {baseClass=&Actor, defaultValue=null, key=GoalActor}
                   }
                  }]
                 properties {children [
                   {childComposite=&self:BTComposite_Selector_6, childTask=null}
                   {childTask=&self:BTTask_Wait_3}
                  ]}
                }
               ]
               decorators [NO AbandonSlot tag? {
                 class=BTDecorator_GameplayTagQuery
                 properties {
                  flowAbortMode=Both
                  nodeName=NO AbandonSlot tag?
                  gameplayTagQuery {
                   autoDescription=" NONE( NPC.Behavior.TargetSlots.Melee.AbandonSlot )"
                   queryTokenStream [0,1,3,1,0]
                   tagDictionary [{tagName=NPC.Behavior.TargetSlots.Melee.AbandonSlot}]
                  }
                 }
                }]
               properties {
                children [
                 {
                  childComposite=null
                  childTask=&self:FortBTTask_ExecuteGameplayAbility_3
                  decorators [
                   &self:BTDecorator_TimeLimit_0
                   &self:FortBTDecorator_GameplayAbility_CompareDistance_3
                   &self:FortBTDecorator_GameplayAbility_HasGameplayAbility_5
                   &self:FortBTDecorator_GameplayAbility_CanActivate_3
                  ]
                 }
                 {
                  childComposite=&self:BTComposite_Selector_13
                  childTask=null
                  decorators [&self:FortBTDecorator_GameplayAbility_CanHitTarget_0]
                 }
                 {
                  childComposite=&self:BTComposite_Sequence_2
                  childTask=null
                  decorators [&self:FortBTDecorator_GameplayAbility_CompareDistance_4]
                 }
                ]
                services []
               }
               services []
              }]
             decorators [Loop {class=BTDecorator_Loop}]
             properties {
              children [{
                childComposite=&self:BTComposite_Selector_12
                childTask=null
                decorators [&self:BTDecorator_GameplayTagQuery_1]
               }]
              services []
             }
             services []
            }]
           decorators [
            NOT Maxed {
             class=BTDecorator_CheckGameplayTagsOnActor
             properties {
              bInverseCondition=true
              flowAbortMode=None
              nodeName=NOT Maxed
              tagsToMatch=Any
              actorToCheck {bNoneIsAllowedValue=false, selectedKeyName=GoalActor}
              gameplayTags {gameplayTags [{tagName=NPC.Behavior.TargetSlots.Melee.Max}]}
             }
            }
            TargetSlots Can Fit? {
             class=TargetSlotsAvailable_C
             properties {
              bCheckConditionOnlyBlackBoardChanges=false
              bInverseCondition=false
              bShowPropertyDetails=true
              customDescription=""
              flowAbortMode=LowerPriority
              nPC PawnParent=null
              nPC PawnParentController=null
              nodeName=TargetSlots Can Fit?
              targetSlotType=Melee
              nPC BlackboardKeyToTestTargetSlotsOn {bNoneIsAllowedValue=false, selectedKeyName=GoalActor}
             }
            }
            MAXED {
             class=BTDecorator_CheckGameplayTagsOnActor
             properties {
              bInverseCondition=false
              flowAbortMode=None
              nodeName=MAXED
              tagsToMatch=Any
              actorToCheck {bNoneIsAllowedValue=false, selectedKeyName=GoalActor}
              gameplayTags {gameplayTags [{tagName=NPC.Behavior.TargetSlots.Melee.Max}]}
             }
            }
            OVERRIDE {
             class=BTDecorator_CheckGameplayTagsOnActor
             properties {
              bInverseCondition=false
              flowAbortMode=None
              nodeName=OVERRIDE
              tagsToMatch=Any
              actorToCheck {bNoneIsAllowedValue=false, selectedKeyName=SelfActor}
              gameplayTags {gameplayTags [{tagName=NPC.Behavior.TargetSlots.Melee.OverrideSlot}]}
             }
            }
            NO ABANDON {
             class=BTDecorator_CheckGameplayTagsOnActor
             properties {
              bInverseCondition=true
              flowAbortMode=None
              nodeName=NO ABANDON
              tagsToMatch=Any
              actorToCheck {bNoneIsAllowedValue=false, selectedKeyName=SelfActor}
              gameplayTags {gameplayTags [{tagName=NPC.Behavior.TargetSlots.Melee.AbandonSlot}]}
             }
            }
           ]
           properties {
            children [{
              childComposite=&self:BTComposite_Selector_11
              childTask=null
              decorators [&self:BTDecorator_Loop_0]
             }]
            services [&self:FortBTService_SendGameplayEvents_C_0]
           }
           services [Send Gameplay Events to Blackboard Key {
             class=FortBTService_SendGameplayEvents_C
             properties {
              activatingPawn=null
              activatingPawnAbilitySystemComponent=null
              bCallTickOnSearchStart=false
              bRestartTimerOnEachActivation=false
              bShowEventDetails=false
              bShowPropertyDetails=true
              customDescription=""
              goalActor=null
              interval=0.5
              nodeName=Send Gameplay Events to Blackboard Key
              ownerActor=null
              randomDeviation=0.1
              targetAbilitySystemComponent=null
              useEventTickLogic=false
              activationEventKeysTags []
              activationEventValuesBlackboards []
              activationEvents {"(TagName=\"NPC.Behavior.TargetSlots.Melee.Add\")" {bNoneIsAllowedValue=false, selectedKeyName=SelfActor}}
              deactivationEventKeysTags []
              deactivationEvents {"(TagName=\"NPC.Behavior.TargetSlots.Melee.Remove\")" {bNoneIsAllowedValue=false, selectedKeyName=SelfActor}}
              deactivationEventsValuesBlackboards []
             }
            }]
          }]
         decorators [
          Has Gameplay Ability - Target Slots Melee {
           class=FortBTDecorator_GameplayAbility_HasGameplayAbility
           properties {
            nodeName=Has Gameplay Ability - Target Slots Melee
            gameplayAbilityTagContainer {defaultValue {gameplayTags [{tagName=NPC.Ability.NonAttack.TargetSlots.Melee}]}}
           }
          }
          Has Gameplay Ability - Melee Primary {
           class=FortBTDecorator_GameplayAbility_HasGameplayAbility
           properties {
            nodeName=Has Gameplay Ability - Melee Primary
            gameplayAbilityTagContainer {defaultValue {gameplayTags [{tagName=NPC.Ability.Attack.Melee.Primary}]}}
           }
          }
          Within Primary Melee Approach Range? {
           class=FortBTDecorator_GameplayAbility_CompareDistance
           properties {
            bInverseCondition=false
            flowAbortMode=LowerPriority
            nodeName=Within Primary Melee Approach Range?
            activeAbilityTagsToSkipTestingValue {key=None, defaultValue {gameplayTags []}}
            distanceComparisons [{
              addedOffset=0
              bUseAddedOffset=false
              bUseOverriddenValue=false
              comparisonType=ThreeDimensions
              operator=Less
              overriddenValue=0
              distanceDataTags {gameplayTags [{tagName=NPC.Ability.MaxApproachRange}]}
             }]
            gameplayAbilityTagContainer {key=None, defaultValue {gameplayTags [{tagName=NPC.Ability.Attack.Melee.Primary}]}}
            targetValue {baseClass=&Actor, defaultValue=null, key=GoalActor}
           }
          }
          Within Primary Melee Approach Range? {
           class=FortBTDecorator_GameplayAbility_CompareDistance
           properties {
            bInverseCondition=false
            flowAbortMode=LowerPriority
            nodeName=Within Primary Melee Approach Range?
            activeAbilityTagsToSkipTestingValue {key=None, defaultValue {gameplayTags []}}
            distanceComparisons [{
              addedOffset=0
              bUseAddedOffset=false
              bUseOverriddenValue=false
              comparisonType=ThreeDimensions
              operator=Greater
              overriddenValue=0
              distanceDataTags {gameplayTags [{tagName=NPC.Ability.MinApproachRange}]}
             }]
            gameplayAbilityTagContainer {key=None, defaultValue {gameplayTags [{tagName=NPC.Ability.Attack.Melee.Primary}]}}
            targetValue {baseClass=&Actor, defaultValue=null, key=GoalActor}
           }
          }
         ]
         properties {
          children [{
            childComposite=&self:BTComposite_Selector_10
            childTask=null
            decoratorOps [
             {number=2, operation=And}
             {number=2, operation=Or}
             {number=2, operation=And}
             {number=0}
             {number=1}
             {number=2, operation=And}
             {number=2}
             {number=3}
             {number=4}
            ]
            decorators [
             &self:BTDecorator_CheckGameplayTagsOnActor_13
             &self:TargetSlotsAvailable_C_1
             &self:BTDecorator_CheckGameplayTagsOnActor_15
             &self:BTDecorator_CheckGameplayTagsOnActor_10
             &self:BTDecorator_CheckGameplayTagsOnActor_14
            ]
           }]
          services [
           &self:FortBTService_ActivateAbility_1
           &self:FortBTService_ActivateAbility_2
          ]
         }
         services [
          Activate Ability - Melee Primary {
           class=FortBTService_ActivateAbility
           properties {
            bCallTickOnSearchStart=false
            bPawnTargetsOnly=false
            bRequireCanHitTargetWithAbility=true
            bRestartTimerOnEachActivation=false
            bShowEventDetails=false
            bShowPropertyDetails=false
            customDescription=""
            interval=0.5
            nodeName=Activate Ability - Melee Primary
            randomDeviation=0
            abilityTags {key=None, defaultValue {gameplayTags [{tagName=NPC.Ability.Attack.Melee.Primary}]}}
            activateOnBecomeRelevant {defaultValue=false, key=None}
            bCanActivateWhenMoving {defaultValue=true, key=None}
            deactivateOnCeaseRelevant {defaultValue=false, key=None}
            distanceChecks [{
              addedOffset=0
              bUseAddedOffset=false
              bUseOverriddenValue=false
              comparisonType=ThreeDimensions
              operator=Less
              overriddenValue=0
              distanceDataTags {gameplayTags [{tagName=NPC.Ability.MaxActualRange}]}
             }]
            initialDelayRandomDeviation {defaultValue=0, key=None}
            initialDelayTime {defaultValue=0, key=None}
            prohibitedTargetTags {gameplayTags []}
            targetActorKey {bNoneIsAllowedValue=false, selectedKeyName=None}
            wantsReceiveActivatedAbilityEnded {defaultValue=false, key=None}
           }
          }
          Activate Ability - Melee Lunge {
           class=FortBTService_ActivateAbility
           properties {
            bCallTickOnSearchStart=false
            bPawnTargetsOnly=false
            bRequireCanHitTargetWithAbility=true
            bRestartTimerOnEachActivation=false
            bShowEventDetails=false
            bShowPropertyDetails=false
            customDescription=""
            interval=0.5
            nodeName=Activate Ability - Melee Lunge
            randomDeviation=0
            abilityTags {key=None, defaultValue {gameplayTags [{tagName=NPC.Ability.Attack.Melee.Lunge}]}}
            activateOnBecomeRelevant {defaultValue=false, key=None}
            bCanActivateWhenMoving {defaultValue=true, key=None}
            deactivateOnCeaseRelevant {defaultValue=false, key=None}
            distanceChecks [
             {
              addedOffset=0
              bUseAddedOffset=false
              bUseOverriddenValue=false
              comparisonType=ThreeDimensions
              operator=Less
              overriddenValue=0
              distanceDataTags {gameplayTags [{tagName=NPC.Ability.MaxActualRange}]}
             }
             {
              addedOffset=0
              bUseAddedOffset=false
              bUseOverriddenValue=false
              comparisonType=ThreeDimensions
              operator=Greater
              overriddenValue=0
              distanceDataTags {gameplayTags [{tagName=NPC.Ability.MinPreferredRange}]}
             }
            ]
            initialDelayRandomDeviation {defaultValue=0, key=None}
            initialDelayTime {defaultValue=0, key=None}
            prohibitedTargetTags {gameplayTags []}
            targetActorKey {bNoneIsAllowedValue=false, selectedKeyName=None}
            wantsReceiveActivatedAbilityEnded {defaultValue=false, key=None}
           }
          }
         ]
        }
        Selector {
         class=BTComposite_Selector
         children [
          GameplayAbility {
           class=FortBTTask_ExecuteGameplayAbility
           decorators [Can Activate Gameplay Ability {
             class=FortBTDecorator_GameplayAbility_CanActivate
             properties {
              flowAbortMode=None
              gameplayAbilityTagContainer {defaultValue {gameplayTags [{tagName=NPC.Ability.NonAttack.Reposition.ForTargetSlots.Melee}]}}
              targetValue {key=AI_Ping_Actor}
             }
            }]
           properties {
            nodeName=GameplayAbility
            gameplayAbilityTagContainer {defaultValue {gameplayTags [{tagName=NPC.Ability.NonAttack.Reposition.ForTargetSlots.Melee}]}}
           }
          }
          Rotate to face BB entry {
           class=BTTask_RotateToFaceBBEntry
           decorators [NOT Rotated To Attack Target {
             class=FortBTDecorator_GameplayAbility_IsRotatedToAttackTarget
             properties {
              bInverseCondition=true
              flowAbortMode=None
              nodeName=NOT Rotated To Attack Target
              activeAbilityTagsToSkipTestingValue {key=None, defaultValue {gameplayTags []}}
              gameplayAbilityTagContainer {key=None, defaultValue {gameplayTags [{tagName=NPC.Ability.Attack.Melee.Primary}]}}
              targetValue {baseClass=&Actor, defaultValue=null, key=GoalActor}
             }
            }]
           properties {blackboardKey {selectedKeyName=GoalActor}, precision {defaultValue=10}}
          }
          Wait {class=BTTask_Wait, properties {randomDeviation {defaultValue=0.25}, waitTime {defaultValue=1}}}
         ]
         decorators [Within Primary Melee Approach Range? {
           class=FortBTDecorator_GameplayAbility_CompareDistance
           properties {
            bInverseCondition=false
            flowAbortMode=LowerPriority
            nodeName=Within Primary Melee Approach Range?
            activeAbilityTagsToSkipTestingValue {key=None, defaultValue {gameplayTags []}}
            distanceComparisons []
            gameplayAbilityTagContainer {key=None, defaultValue {gameplayTags [{tagName=NPC.Ability.Attack.Melee.Primary}]}}
            targetValue {baseClass=&Actor, defaultValue=null, key=GoalActor}
           }
          }]
         properties {
          children [
           {
            childComposite=null
            childTask=&self:FortBTTask_ExecuteGameplayAbility_0
            decorators [&self:FortBTDecorator_GameplayAbility_CanActivate_0]
           }
           {
            childComposite=null
            childTask=&self:BTTask_RotateToFaceBBEntry_0
            decorators [&self:FortBTDecorator_GameplayAbility_IsRotatedToAttackTarget_0]
           }
           {childComposite=null, childTask=&self:BTTask_Wait_9, decorators []}
          ]
          services [&self:FortBTService_ActivateAbility_0]
         }
         services [Activate Ability - Melee Primary {
           class=FortBTService_ActivateAbility
           properties {
            bCallTickOnSearchStart=false
            bPawnTargetsOnly=false
            bRequireCanHitTargetWithAbility=true
            bRestartTimerOnEachActivation=false
            bShowEventDetails=false
            bShowPropertyDetails=false
            customDescription=""
            interval=0.5
            nodeName=Activate Ability - Melee Primary
            randomDeviation=0
            abilityTags {key=None, defaultValue {gameplayTags [{tagName=NPC.Ability.Attack.Melee.Primary}]}}
            activateOnBecomeRelevant {defaultValue=false, key=None}
            bCanActivateWhenMoving {defaultValue=true, key=None}
            deactivateOnCeaseRelevant {defaultValue=false, key=None}
            distanceChecks [{
              addedOffset=0
              bUseAddedOffset=false
              bUseOverriddenValue=false
              comparisonType=ThreeDimensions
              operator=Less
              overriddenValue=0
              distanceDataTags {gameplayTags [{tagName=NPC.Ability.MaxActualRange}]}
             }]
            initialDelayRandomDeviation {defaultValue=0, key=None}
            initialDelayTime {defaultValue=0, key=None}
            prohibitedTargetTags {gameplayTags []}
            targetActorKey {bNoneIsAllowedValue=false, selectedKeyName=None}
            wantsReceiveActivatedAbilityEnded {defaultValue=false, key=None}
           }
          }]
        }
        FortMoveTo {
         class=FortBTTask_GameMoveTo
         properties {
          bDeimosFlavor=false
          bFinishMoveOnOverlap=true
          bStopAtGoal=false
          nodeName=FortMoveTo
          acceptableRadius {defaultValue=30}
          bReachTestIncludesAgentRadius {defaultValue=true}
          bReachTestIncludesGoalRadius {defaultValue=true}
          blackboardKey {selectedKeyName=GoalActor}
          filterClass {defaultValue=&NavFilter_NPC_AnyTraversable_C}
          navFilterTag {tagName=NavFilter.NPC.DefaultMove}
         }
        }
       ]
       decorators [Does path exist {
         class=BTDecorator_DoesPathExist
         properties {
          bInverseCondition=false
          flowAbortMode=None
          nodeName=Does path exist
          blackboardKeyA {bNoneIsAllowedValue=false, selectedKeyName=SelfActor}
          blackboardKeyB {bNoneIsAllowedValue=false, selectedKeyName=GoalActor}
          filterClass {baseClass=&NavigationQueryFilter, defaultValue=null, key=None}
          pathQueryType {defaultValue=2, enumType=/Script/AIModule.EPathExistanceQueryType, key=None, nativeEnumTypeName=EPathExistanceQueryType}
         }
        }]
       properties {
        children [
         {
          childComposite=&self:BTComposite_Selector_9
          childTask=null
          decorators [
           &self:FortBTDecorator_GameplayAbility_HasGameplayAbility_3
           &self:FortBTDecorator_GameplayAbility_HasGameplayAbility_4
           &self:FortBTDecorator_GameplayAbility_CompareDistance_1
           &self:FortBTDecorator_GameplayAbility_CompareDistance_2
          ]
         }
         {
          childComposite=&self:BTComposite_Selector_5
          childTask=null
          decorators [&self:FortBTDecorator_GameplayAbility_CompareDistance_0]
         }
         {childComposite=null, childTask=&self:FortBTTask_GameMoveTo_0, decorators []}
        ]
        services []
       }
       services []
      }
      GameplayAbility {
       class=FortBTTask_ExecuteGameplayAbility
       decorators [
        Has Gameplay Ability {
         class=FortBTDecorator_GameplayAbility_HasGameplayAbility
         properties {gameplayAbilityTagContainer {defaultValue {gameplayTags [{tagName=NPC.Ability.NonAttack.PlayMontage.Wander.NearGoal}]}}}
        }
        Can Activate Gameplay Ability {
         class=FortBTDecorator_GameplayAbility_CanActivate
         properties {
          flowAbortMode=None
          gameplayAbilityTagContainer {defaultValue {gameplayTags [{tagName=NPC.Ability.NonAttack.PlayMontage.Wander.NearGoal}]}}
          targetValue {key=AI_Ping_Actor}
         }
        }
        Gameplay Tag Query {
         class=BTDecorator_GameplayTagQuery
         properties {
          flowAbortMode=None
          nodeName=Gameplay Tag Query
          gameplayTagQuery {
           autoDescription=" ANY( NPC.Behavior.Phase.Wander.NearGoal.Animation.Pending )"
           queryTokenStream [0,1,1,1,0]
           tagDictionary [{tagName=NPC.Behavior.Phase.Wander.NearGoal.Animation.Pending}]
          }
         }
        }
       ]
       properties {
        nodeName=GameplayAbility
        gameplayAbilityTagContainer {defaultValue {gameplayTags [{tagName=NPC.Ability.NonAttack.PlayMontage.Wander.NearGoal}]}}
       }
      }
      GameplayAbility {
       class=FortBTTask_ExecuteGameplayAbility
       decorators [
        Has Gameplay Ability {
         class=FortBTDecorator_GameplayAbility_HasGameplayAbility
         properties {gameplayAbilityTagContainer {defaultValue {gameplayTags [{tagName=NPC.Ability.NonAttack.Wander.NearGoal}]}}}
        }
        Can Activate Gameplay Ability {
         class=FortBTDecorator_GameplayAbility_CanActivate
         properties {
          flowAbortMode=LowerPriority
          gameplayAbilityTagContainer {defaultValue {gameplayTags [{tagName=NPC.Ability.NonAttack.Wander.NearGoal}]}}
          targetValue {key=AI_Ping_Actor}
         }
        }
       ]
       properties {nodeName=GameplayAbility, gameplayAbilityTagContainer {defaultValue {gameplayTags [{tagName=NPC.Ability.NonAttack.Wander.NearGoal}]}}}
      }
     ]
     decorators [
      Is Goal Pawn ? {
       class=FortBTDecorator_IsGoalPawn
       properties {bInverseCondition=false, flowAbortMode=Both, nodeName=Is Goal Pawn ?, blackboardKey {bNoneIsAllowedValue=false, selectedKeyName=GoalActor}}
      }
      Has any attack ability ? {
       class=FortBTDecorator_GameplayAbility_HasGameplayAbility
       properties {
        flowAbortMode=Both
        nodeName=Has any attack ability ?
        gameplayAbilityTagContainer {defaultValue {gameplayTags [{tagName=NPC.Ability.Attack}]}}
       }
      }
      Distance Between PingLocation - GoalActor {
       class=FortBTDecorator_DistanceBetween
       properties {
        flowAbortMode=Both
        nodeName=Distance Between PingLocation - GoalActor
        operator=LessOrEqual
        blackboardKeyA {selectedKeyName=AI_Ping_Location}
        blackboardKeyB {selectedKeyName=GoalActor}
        specifiedDistance {defaultValue=3000}
       }
      }
     ]
     properties {
      children [
       {
        childComposite=&self:BTComposite_Selector_8
        childTask=null
        decorators [&self:BTDecorator_DoesPathExist_1]
       }
       {
        childComposite=null
        childTask=&self:FortBTTask_ExecuteGameplayAbility_1
        decorators [
         &self:FortBTDecorator_GameplayAbility_HasGameplayAbility_0
         &self:FortBTDecorator_GameplayAbility_CanActivate_1
         &self:BTDecorator_GameplayTagQuery_0
        ]
       }
       {
        childComposite=null
        childTask=&self:FortBTTask_ExecuteGameplayAbility_2
        decorators [
         &self:FortBTDecorator_GameplayAbility_HasGameplayAbility_2
         &self:FortBTDecorator_GameplayAbility_CanActivate_2
        ]
       }
      ]
      services []
     }
     services []
    }
    Sequence {
     class=BTComposite_Sequence
     children [
      Move To {
       class=BTTask_MoveTo
       children []
       decorators []
       properties {
        bIgnoreRestartSelf=false
        nodeName=Move To
        acceptableRadius {defaultValue=5, key=None}
        bAllowPartialPath {defaultValue=true, key=None}
        bAllowStrafe {defaultValue=false, key=None}
        bProjectGoalLocation {defaultValue=true, key=None}
        bReachTestIncludesAgentRadius {defaultValue=true, key=None}
        bReachTestIncludesGoalRadius {defaultValue=true, key=None}
        bRequireNavigableEndLocation {defaultValue=true, key=None}
        bStartFromPreviousPath {defaultValue=false, key=None}
        bTrackMovingGoal {defaultValue=true, key=None}
        blackboardKey {bNoneIsAllowedValue=false, selectedKeyName=AI_Ping_Location}
        filterClass {baseClass=&NavigationQueryFilter, defaultValue=null, key=None}
        observedBlackboardValueTolerance {defaultValue=4.75, key=None}
       }
       services []
      }
      Wait {class=BTTask_Wait, properties {waitTime {defaultValue=3600}}}
     ]
     decorators []
     properties {children [
       {childTask=&self:BTTask_MoveTo_0}
       {childTask=&self:BTTask_Wait_22}
      ]}
    }
   ]
   decorators [
    Ping GoTo ? {
     class=BTDecorator_Blackboard
     properties {arithmeticOperation=Equal, intValue=1, nodeName=Ping GoTo ?, stringValue=PingGoTo, blackboardKey {selectedKeyName=AI_Ping_Command}}
    }
    Location ? {
     class=BTDecorator_Blackboard
     properties {
      arithmeticOperation=Equal
      intValue=1
      nodeName=Location ?
      notifyObserver=ValueChange
      stringValue=PingGoTo
      blackboardKey {selectedKeyName=AI_Ping_Location}
     }
    }
   ]
   properties {
    children [
     {
      childComposite=&self:BTComposite_Selector_1
      childTask=null
      decorators [
       &self:FortBTDecorator_IsGoalPawn_0
       &self:FortBTDecorator_GameplayAbility_HasGameplayAbility_1
       &self:FortBTDecorator_DistanceBetween_0
      ]
     }
     {childComposite=&self:BTComposite_Sequence_1, childTask=null, decorators []}
    ]
    services []
   }
   services []
  }
  Wait {
   class=BTTask_Wait
   decorators [Ping Hold Polistion ? {
     class=BTDecorator_Blackboard
     properties {
      arithmeticOperation=Equal
      intValue=3
      nodeName=Ping Hold Polistion ?
      stringValue=PingHoldPosition
      blackboardKey {selectedKeyName=AI_Ping_Command}
     }
    }]
   properties {waitTime {defaultValue=3600}}
  }
  Wait {class=BTTask_Wait, properties {waitTime {defaultValue=3600}}}
 ]
 decorators []
 properties {
  children [
   {
    childComposite=&self:BTComposite_Selector_3
    childTask=null
    decorators [
     &self:BTDecorator_Blackboard_5
     &self:BTDecorator_Blackboard_6
    ]
   }
   {
    childComposite=null
    childTask=&self:BTTask_Wait_19
    decorators [&self:BTDecorator_Blackboard_4]
   }
   {childComposite=null, childTask=&self:BTTask_Wait_23, decorators []}
  ]
  services []
 }
 services []
}
