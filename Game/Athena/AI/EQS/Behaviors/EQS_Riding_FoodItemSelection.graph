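# EQS query that picks a food loot goal for AI, named for riding behavior.
# A LootGoals generator gathers available loot in a box around the querier as Monster-team world Loot assignments.
# A gameplay-tag test, commented as finding foraged items, filters them by a query looked up from the loot tag, passing when none exists.
# A distance test drops far goals and scores nearer ones higher with clamped inverse-linear scoring.
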EnvQuery /Game/Athena/AI/EQS/Behaviors/EQS_Riding_FoodItemSelection
options [{
  generator {
   name=FortQueryGenerator_LootGoalsAthena_0
   class=FortQueryGenerator_LootGoalsAthena
   properties {
    assignmentSettings=/Game/Athena/AI/GoalSelection/Pawn/LootAssignmentSettings_Default
    bAutoSortTests=true
    bAvailableLootOnly=true
    bCanRunAsync=false
    envQueryResultNormalizationOption=Default
    horizontalHalfExtents=1200
    optionName=""
    searchCenter=&EnvQueryContext_Querier
    verticalHalfExtents=1200
    assignmentIdentifier {assignmentTeam=Monster, assignmentType=World, assignmentGameplayTags {gameplayTags [{tagName=NPC.Assignment.Loot}]}}
    excludedAILootGameplayTags {gameplayTags []}
   }
  }
  tests [
   FortQueryTest_GoalGameplayTags_0 {
    class=FortQueryTest_GoalGameplayTags
    properties {
     bDefineReferenceValue=false
     bRequireAllProvidedTagQueriesPass=false
     bScoreEncounterGoals=false
     bScoreEnemies=false
     bScoreSpecificAssignments=true
     bScoreWorldGoals=false
     bShouldLookupQueryByTag=true
     bShouldPassWhenQueryNotFound=true
     boolValue=true
     clampMaxType=None
     clampMinType=None
     filterType=Match
     floatValueMax=0
     floatValueMin=0
     multipleContextFilterOp=AllPass
     multipleContextScoreOp=AverageScore
     normalizationType=Absolute
     referenceValue=0
     scoreClampMax=0
     scoreClampMin=0
     scoringEquation=Constant
     scoringFactor=1
     testComment=Find Foraged Items
     testPurpose=Filter
     assignmentIds [{assignmentTeam=Monster, assignmentType=World, assignmentGameplayTags {gameplayTags [{tagName=NPC.Assignment.Loot}]}}]
     goalActorTagQuery {autoDescription="", tokenStreamVersion=0, userDescription="", queryTokenStream [], tagDictionary []}
     pawnOwnerTagQuery {autoDescription="", tokenStreamVersion=0, userDescription="", queryTokenStream [], tagDictionary []}
     prohibitedAssignmentIds []
     queryLookupTag {tagName=NPC.Assignment.Loot}
     tagQueryToMatch {autoDescription="", tokenStreamVersion=0, userDescription="", queryTokenStream [], tagDictionary []}
    }
   }
   FortQueryTest_GoalDistance_1 {
    class=FortQueryTest_GoalDistance
    properties {
     bScoreEncounterGoals=false
     bScoreEnemies=false
     bScoreSpecificAssignments=true
     bScoreWorldGoals=false
     clampMaxType=SpecifiedValue
     clampMinType=SpecifiedValue
     filterType=Maximum
     floatValueMax=3000
     scoreClampMax=5000
     scoreClampMin=200
     scoringEquation=InverseLinear
     scoringFactor=1
     testComment=Loot Scoring and Filtering
     testPurpose=FilterAndScore
     assignmentIds [{assignmentTeam=Monster, assignmentType=World, assignmentGameplayTags {gameplayTags [{tagName=NPC.Assignment.Loot}]}}]
    }
   }
  ]
 }]
